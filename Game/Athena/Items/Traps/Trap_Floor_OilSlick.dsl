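; Trap_Floor_OilSlick is the Blueprint for a placeable floor oil slick trap in the Athena mode.
; On its placed event, a DoOnce gate registers the trap's Trigger component and plays the place-trap sound at the actor's location.
; The construction script is empty, and no slick or trigger-response logic appears in this graph.
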(event EventOnPlaced
  (Utilities|FlowControl|DoOnce false
    (:Completed
      (Trap|AddTriggerComponent (Variables|Default|GetTrigger))
      (Audio|PlaySoundatLocation (Variables|Audio|GetPlace%20Trap%20Sound) (Transformation|GetActorLocation)))))

(fn ConstructionScript)
